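# Sound cue for the Sahara back bling's charge sound, a reactive cosmetic effect.
# A Random node picks with equal weight between two wave variations, Backpack_Sahara_01 and Backpack_Sahara_02.
# It routes through the Backblings sound class and reactive backbling property sheet, using the close-range pets attenuation with reduced volume.
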SoundCue /BRCosmetics/Sounds/Backblings/Sahara/Backpack_Sahara_Charge_Cue
root=Random_0
nodes [
 {id=Random_0, children [WavePlayer_0,WavePlayer_1], properties {weights [1,1]}}
 {id=WavePlayer_0, properties {soundWaveAssetPtr=&/Backpack_Sahara_01}}
 {id=WavePlayer_1, properties {soundWaveAssetPtr=&/Backpack_Sahara_02}}
]
properties {
 attenuationSettings=&Fort_Pets_Close_Athena_Att
 audioPropertiesSheet=/BRCosmetics/Sounds/Settings/PropertySheets/Reactive/APS_Cosmetic_Reactive_Backbling
 maxDistance=1800.0001
 soundClassObject=/Game/Sounds/FortSoundClassesAndModes/Fort_Backblings
 volumeMultiplier=0.63
}
